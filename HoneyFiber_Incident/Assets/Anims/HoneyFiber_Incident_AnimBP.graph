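# Animation Blueprint for the HoneyFiber Incident consumable weapon that shows a closed pose at full ammo and an open pose otherwise.
# Each update casts the owning actor to FortWeapon and sets MaxAmmo by comparing remaining ammo to the item definition's initial ammo.
# The AnimGraph uses that bool to blend linearly between the two pose sequences, then passes the result through DefaultSlot for montage playback.
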AnimBlueprint /HoneyFiber_Incident/Assets/Anims/HoneyFiber_Incident_AnimBP
target_skeleton=/HoneyFiber_Incident/Assets/Meshes/HoneyFiber_Consumable_Skeleton
graphs [
 AnimGraph {nodes [
   n1 = Sequence Player 'HoneyFiber_Incident_Pose_Closed_WEP'() {sequence=/HoneyFiber_Incident/Gameplay/Weapons/HoneyFiber_Incident/Animation/HoneyFiber_Incident_Pose_Closed_WEP}
   n2 = Sequence Player 'HoneyFiber_Incident_Pose_Open_WEP'() {sequence=/HoneyFiber_Incident/Gameplay/Weapons/HoneyFiber_Incident/Animation/HoneyFiber_Incident_Pose_Open_WEP}
   n3 = Get maxAmmo() -> (maxAmmo:bool)
   Output Pose(Result=Slot 'DefaultSlot'(Source=Blend Poses by bool(BlendPose_0=@n1, BlendPose_1=@n2, bActiveValue=@n3) {blendType=Linear}))
  ]}
 EventGraph {nodes [
   Event BlueprintUpdateAnimation()
   > n1 = Cast To FortWeapon!(execute, Object=GetOwningActor()) -> (then, CastFailed, AsFort Weapon:Fort Weapon Object Reference) {is=title K2Node_DynamicCast}
   n2 = GetWeaponData!(self:Fort Weapon Object Reference=@n1) -> (ReturnValue:Fort Weapon Item Definition Object Reference)
   > Set MaxAmmo(MaxAmmo:bool="Equal (Integer)"(
     A=GetRemainingAmmo!(self:Fort Weapon Object Reference=@n1) -> (ReturnValue:int=0)
     B=GetInitialAmmo!(self:Fort World Item Definition Object Reference=@n2, ItemLevel:int=0) -> (ReturnValue:int=0)
    )) -> (Output_Get:bool)
  ]}
]
variables [MaxAmmo {type=boolean}]
