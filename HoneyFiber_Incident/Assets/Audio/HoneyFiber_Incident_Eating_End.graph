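# SoundCue for the end-of-eating sound of a HoneyFiber Incident consumable.
# A Random node picks evenly among three "Consumable_Finished" wave variants, and a Modulator adds slight random pitch variation.
# It uses Fortnite consumable attenuation and sound class, an emote foley concurrency set, and a send to a music audio analysis submix.
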SoundCue /HoneyFiber_Incident/Assets/Audio/HoneyFiber_Incident_Eating_End
root=Modulator_0
nodes [
 {id=Modulator_0, children [Random_2], properties {pitchMax=1.03, pitchMin=0.94}}
 {id=Random_2, children [WavePlayer_6,WavePlayer_7,WavePlayer_8], properties {weights [1,1,1]}}
 {id=WavePlayer_6, properties {soundWaveAssetPtr=/HoneyFiber_Incident/Assets/Audio/waves/HoneyFiberIncident_Consumable_Finished_01}}
 {id=WavePlayer_7, properties {soundWaveAssetPtr=/HoneyFiber_Incident/Assets/Audio/waves/HoneyFiberIncident_Consumable_Finished_02}}
 {id=WavePlayer_8, properties {soundWaveAssetPtr=/HoneyFiber_Incident/Assets/Audio/waves/HoneyFiberIncident_Consumable_Finished_03}}
]
properties {
 attenuationSettings=&Fort_Consumable_Use_Athena_Att
 maxDistance=2250
 soundClassObject=&Fort_Consumables
 concurrencyOverrides {maxCount=3, resolutionRule=StopOldest}
 concurrencySet [/Game/Sounds/Concurrency_Rules/Athena_EmoteFoley_Conc]
 soundSubmixSends [{minSendLevel=0, soundSubmix=&Fortnite_AudioAnalysis_Music}]
}
