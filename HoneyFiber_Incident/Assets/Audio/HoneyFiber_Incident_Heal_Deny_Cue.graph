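# Sound cue for a heal-deny feedback sound in HoneyFiber_Incident, playing a stock Fortnite negative UI sound.
# A single wave player feeds the root directly, at a reduced volume.
# Playback uses the HoneyFiber impact attenuation, a third-person ranged weapon fire sound class and an overridden concurrency with a small voice cap.
# It also carries a send to the music audio-analysis submix.
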SoundCue /HoneyFiber_Incident/Assets/Audio/HoneyFiber_Incident_Heal_Deny_Cue
root=WavePlayer_1
nodes [{id=WavePlayer_1, properties {soundWaveAssetPtr=/Game/Sounds/Fort_UI_Sounds/Gameplay/fort_ui_new_neg_01}}]
properties {
 attenuationSettings=&/HoneyFiber_impact_Att
 bOverrideConcurrency=true
 maxDistance=4300
 soundClassObject=&Fort_WeaponRanged_Fire_3P_PSM
 volumeMultiplier=0.2
 concurrencyOverrides {maxCount=4}
 soundSubmixSends [{minSendLevel=0, soundSubmix=&Fortnite_AudioAnalysis_Music}]
}
