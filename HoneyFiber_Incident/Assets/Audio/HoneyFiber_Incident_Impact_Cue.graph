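# Sound cue for the HoneyFiber Incident impact hit, classed as third-person ranged weapon fire audio.
# It randomly picks one of three equally weighted neutral hit waves, varies it through a Modulator, and plays it after a short fixed delay.
# A custom impact attenuation shapes it in space, a concurrency override allows one instance at a time, and a submix send feeds music audio analysis.
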SoundCue /HoneyFiber_Incident/Assets/Audio/HoneyFiber_Incident_Impact_Cue
root=Delay_1
nodes [
 {id=Delay_1, children [Modulator_0], properties {delayMax=0.05, delayMin=0.05}}
 {id=Modulator_0, children [Random_1]}
 {id=Random_1, children [WavePlayer_4,WavePlayer_5,WavePlayer_6], properties {weights [1,1,1]}}
 {id=WavePlayer_4, properties {soundWaveAssetPtr=/HoneyFiber_Incident/Assets/Audio/waves/HoneyFiberIncident_Hit_Neutral_01}}
 {id=WavePlayer_5, properties {soundWaveAssetPtr=/HoneyFiber_Incident/Assets/Audio/waves/HoneyFiberIncident_Hit_Neutral_02}}
 {id=WavePlayer_6, properties {soundWaveAssetPtr=/HoneyFiber_Incident/Assets/Audio/waves/HoneyFiberIncident_Hit_Neutral_03}}
]
properties {
 attenuationSettings=&/HoneyFiber_impact_Att
 bOverrideConcurrency=true
 maxDistance=4300
 soundClassObject=&Fort_WeaponRanged_Fire_3P_PSM
 volumeMultiplier=0.7
 concurrencyOverrides {maxCount=1}
 soundSubmixSends [{minSendLevel=0, soundSubmix=&Fortnite_AudioAnalysis_Music}]
}
