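# Sound cue that plays damage impact hit sounds for HoneyFiber_Incident.
# An equally weighted Random node picks one of three hit waves and feeds a default Modulator, with volume reduced at cue level.
# Overridden attenuation adds a custom falloff curve, distance low-pass, listener focus and a distant-reverb send.
# It caps concurrency at four, uses a third-person ranged-weapon-fire sound class and also sends to the music audio-analysis submix.
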SoundCue /HoneyFiber_Incident/Assets/Audio/HoneyFiber_Incident_Impact_Damage_Cue
root=Modulator_0
nodes [
 {id=Modulator_0, children [Random_0]}
 {id=Random_0, children [WavePlayer_0,WavePlayer_1,WavePlayer_2], properties {weights [1,1,1]}}
 {id=WavePlayer_0, properties {soundWaveAssetPtr=/HoneyFiber_Incident/Assets/Audio/waves/HoneyFiberIncident_Hit_Damage_01}}
 {id=WavePlayer_1, properties {soundWaveAssetPtr=/HoneyFiber_Incident/Assets/Audio/waves/HoneyFiberIncident_Hit_Damage_02}}
 {id=WavePlayer_2, properties {soundWaveAssetPtr=/HoneyFiber_Incident/Assets/Audio/waves/HoneyFiberIncident_Hit_Damage_03}}
]
properties {
 attenuationSettings=&/HoneyFiber_impact_Att
 bOverrideAttenuation=true
 bOverrideConcurrency=true
 maxDistance=5000
 soundClassObject=&Fort_WeaponRanged_Fire_3P_PSM
 volumeMultiplier=0.4
 attenuationOverrides {
  bAttenuateWithLPF=true
  bEnableFocusInterpolation=true
  bEnableListenerFocus=true
  bEnableSubmixSends=true
  distanceAlgorithm=Custom
  falloffDistance=4000
  focusVolumeAttenuation=0.8
  lPFFrequencyAtMax=2000
  lPFRadiusMax=2000
  lPFRadiusMin=500
  reverbDistanceMax=1024
  reverbDistanceMin=0
  reverbWetLevelMax=0
  reverbWetLevelMin=1
  stereoSpread=100
  attenuationShapeExtents {x=1000}
  customAttenuationCurve {editorCurveData {keys [
     <at 0 1 RCIM_Cubic>
     {arriveTangent=-1.9317, interpMode=RCIM_Cubic, leaveTangent=-1.9317, tangentMode=RCTM_User, time=0.1329, value=0.8199}
     {arriveTangent=-1.4258, interpMode=RCIM_Cubic, leaveTangent=-1.4258, time=0.2845, value=0.345}
     {arriveTangent=-0.1901, interpMode=RCIM_Cubic, leaveTangent=-0.1901, tangentMode=RCTM_User, time=0.6444, value=0.0906}
     {arriveTangent=-0.2549, interpMode=RCIM_Cubic, leaveTangent=-0.2549, time=0.7516, value=0.0752}
     {arriveTangent=-0.7607, interpMode=RCIM_Cubic, leaveTangent=-0.7607, tangentMode=RCTM_User, time=1, value=0}
    ]}}
  submixSendSettings [{
    bEnableHPFCutoff=false
    bEnableLPFCutoff=false
    disableManualSendClamp=false
    hPFCutoff=20
    lPFCutoff=20000
    maxSendDistance=5000
    maxSendLevel=0.18
    minSendDistance=1000
    minSendLevel=0.01
    sendLevel=0.2
    sendLevelControlMethod=Linear
    soundSubmix=&Fortnite_Reverb_Distant
    customSendLevelCurve {externalCurve=/Game/Sounds/SubMixes/Reverb/Fortnite_Reverb_Distant_Send}
   }]
 }
 concurrencyOverrides {maxCount=4}
 soundSubmixSends [{minSendLevel=0, soundSubmix=&Fortnite_AudioAnalysis_Music}]
}
