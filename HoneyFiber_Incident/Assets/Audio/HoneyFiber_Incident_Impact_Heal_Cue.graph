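# Sound cue for the HoneyFiber Incident heal impact sound, played when a heal hit lands.
# A Random node picks one of six equally weighted Hit_Heal wave variations and feeds it through a Modulator at default settings.
# The cue uses a HoneyFiber impact attenuation asset, a third-person ranged weapon fire sound class, and a concurrency override capping simultaneous instances.
# It also sends to the music audio-analysis submix.
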SoundCue /HoneyFiber_Incident/Assets/Audio/HoneyFiber_Incident_Impact_Heal_Cue
root=Modulator_0
nodes [
 {id=Modulator_0, children [Random_0]}
 {id=Random_0, children [WavePlayer_0,WavePlayer_1,WavePlayer_2,WavePlayer_3,WavePlayer_4,WavePlayer_5], properties {weights [1,1,1,1,1,1]}}
 {id=WavePlayer_0, properties {soundWaveAssetPtr=/HoneyFiber_Incident/Assets/Audio/waves/HoneyFiberIncident_Hit_Heal_01}}
 {id=WavePlayer_1, properties {soundWaveAssetPtr=/HoneyFiber_Incident/Assets/Audio/waves/HoneyFiberIncident_Hit_Heal_02}}
 {id=WavePlayer_2, properties {soundWaveAssetPtr=/HoneyFiber_Incident/Assets/Audio/waves/HoneyFiberIncident_Hit_Heal_03}}
 {id=WavePlayer_3, properties {soundWaveAssetPtr=/HoneyFiber_Incident/Assets/Audio/waves/HoneyFiberIncident_Hit_Heal_04}}
 {id=WavePlayer_4, properties {soundWaveAssetPtr=/HoneyFiber_Incident/Assets/Audio/waves/HoneyFiberIncident_Hit_Heal_05}}
 {id=WavePlayer_5, properties {soundWaveAssetPtr=/HoneyFiber_Incident/Assets/Audio/waves/HoneyFiberIncident_Hit_Heal_06}}
]
properties {
 attenuationSettings=&/HoneyFiber_impact_Att
 bOverrideConcurrency=true
 maxDistance=4300
 soundClassObject=&Fort_WeaponRanged_Fire_3P_PSM
 volumeMultiplier=0.8
 concurrencyOverrides {maxCount=4}
 soundSubmixSends [{minSendLevel=0, soundSubmix=&Fortnite_AudioAnalysis_Music}]
}
