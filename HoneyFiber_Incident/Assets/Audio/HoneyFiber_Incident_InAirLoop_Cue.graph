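# Looping in-air sound cue for the HoneyFiber_Incident asset set, using projectile in-air attenuation and a PreventNew concurrency override.
# Two looping waves, a travel whistle and an in-air loop, are mixed and enveloped to fade in over about half a second.
# A top-level modulator randomizes pitch per play.
# The low-volume Random branch in the final mixer has only empty inputs, so it adds no sound.
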SoundCue /HoneyFiber_Incident/Assets/Audio/HoneyFiber_Incident_InAirLoop_Cue
root=Modulator_0
nodes [
 {
  id=Enveloper_1
  children [Mixer_1]
  properties {
   pitchCurve {editorCurveData {keys [{time=0}]}}
   volumeCurve {editorCurveData {keys [<at 0.0001 0.0145>,<at 0.1225 0.0867>,<at 0.535 0.9683>]}}
  }
 }
 {id=Mixer_0, children [Enveloper_1,Random_0], properties {inputVolume [2,0.1]}}
 {id=Mixer_1, children [WavePlayer_0,WavePlayer_1], properties {inputVolume [2,2]}}
 {id=Modulator_0, children [Mixer_0], properties {pitchMax=1.15, pitchMin=0.9}}
 {id=Random_0, children [null,null,null], properties {weights [1,1,1]}}
 {id=WavePlayer_0, properties {bLooping=true, soundWaveAssetPtr=&/HoneyFiber_Incident_Travel_Whistle_Loop}}
 {id=WavePlayer_1, properties {bLooping=true, soundWaveAssetPtr=/HoneyFiber_Incident/Assets/Audio/waves/HoneyFiberIncident_InAir_Loop}}
]
properties {
 attenuationSettings=/Game/Sounds/Attenuation/Fort_Weapon_Projectile_InAir_Loop_Att
 bOverrideConcurrency=true
 maxDistance=4500
 soundClassObject=&Fort_WeaponRanged_Fire_3P_PSM
 volumeMultiplier=1
 attenuationOverrides {falloffDistance=5000, reverbWetLevelMax=0.7, reverbWetLevelMin=0.2, attenuationShapeExtents {x=1000}}
 concurrencyOverrides {resolutionRule=PreventNew}
}
