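# One-shot mono MetaSound source preset that plays the drop sound for the HoneyFiber Incident item.
# It wraps the shared MSS_Item_Drop graph, feeding it three drop waves, master volume and the play trigger, and passing its audio and finished trigger out.
# Random pitch-shift base and spread inputs are declared but not wired into that graph.
# Fortnite consumable attenuation and sound class, plus item concurrency, govern playback.
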MetaSoundSource /HoneyFiber_Incident/Assets/Audio/MSS_HoneyFiber_Incident_Drop_Preset
inputs [
 MasterVolume {type=Float}
 Random.PitchShift.Base {type=Float}
 Random.PitchShift.Spread {type=Float}
 Sounds {type=WaveAsset:Array}
 UE.Source.OnPlay {type=Trigger}
]
interfaces [UE.OutputFormat.Mono:1.0, UE.Source.OneShot:1.0, UE.Source:1.1]
outputs [
 "UE.OutputFormat.Mono.Audio:0" {type=Audio}
 UE.Source.OneShot.OnFinished {type=Trigger}
]
pages [Default {
  input_defaults {
   MasterVolume=1
   Random.PitchShift.Base=0
   Random.PitchShift.Spread=0.5
   UE.Source.OnPlay=null
   Sounds [
    /HoneyFiber_Incident/Assets/Audio/waves/HoneyFiberIncident_Drop_01
    /HoneyFiber_Incident/Assets/Audio/waves/HoneyFiberIncident_Drop_02
    /HoneyFiber_Incident/Assets/Audio/waves/HoneyFiberIncident_Drop_03
   ]
  }
  nodes [MSS_Item_Drop_0 = /FortMetasound/Items/MSS_Item_Drop(MasterVolume=^MasterVolume, Sounds=^Sounds, UE.Source.OnPlay=^UE.Source.OnPlay)]
  links [
   {dst="@UE.OutputFormat.Mono.Audio:0", src=MSS_Item_Drop_0.UE.OutputFormat.Mono.Audio:0}
   {dst="@UE.Source.OneShot.OnFinished", src=MSS_Item_Drop_0.UE.Source.OneShot.OnFinished}
  ]
 }]
properties {
 assetImportData=&self:AssetImportData
 attenuationSettings=&Fort_Consumable_Use_Athena_Att
 maxDistance=2250
 sampleRate=48000
 soundClassObject=&Fort_Consumables
 volume=0.4
 concurrencySet [&CS_Fort_Item]
}
