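# Sound cue for the intro of the eating sound in the HoneyFiber_Incident consumable set.
# A single wave player plays one intro eating wave directly, with no randomization or mixing.
# It uses the Athena consumable-use attenuation with a limited audible range and routes through the consumables sound class.
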SoundCue /HoneyFiber_Incident/Assets/Audio/SC_HoneyFiberIncident_Eating_Intro_Cue
root=WavePlayer_0
nodes [{id=WavePlayer_0, properties {soundWaveAssetPtr=/HoneyFiber_Incident/Assets/Audio/waves/HoneyFiberIncident_Eating_Intro_01}}]
properties {
 attenuationSettings=&Fort_Consumable_Use_Athena_Att
 maxDistance=2250
 soundClassObject=&Fort_Consumables
 volumeMultiplier=1
}
