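# Sound cue for the HoneyFiber Incident eating loop, routed to the consumables sound class with consumable-use attenuation.
# A Mixer layers a non-looping heal start sound at fixed reduced volume, a looping eating wave, and the default player healing loop at lower mix level.
# An Enveloper fades the healing loop in over the first second and raises its pitch over twelve seconds, suggesting progress during use.
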SoundCue /HoneyFiber_Incident/Assets/Audio/SC_HoneyFiberIncident_Eating_Loop_Cue
root=Mixer_1
nodes [
 {
  id=Enveloper_1
  children [WavePlayer_1]
  properties {pitchCurve {editorCurveData {keys [<at 0 0.65>,<at 12 2>]}}, volumeCurve {editorCurveData {keys [<at 0 0>,<at 1 1>]}}}
 }
 {id=Mixer_1, children [Modulator_0,Enveloper_1,WavePlayer_0], properties {inputVolume [2,0.5,2]}}
 {id=Modulator_0, children [WavePlayer_3], properties {volumeMax=0.6, volumeMin=0.6}}
 {id=WavePlayer_0, properties {bLooping=true, soundWaveAssetPtr=/HoneyFiber_Incident/Assets/Audio/waves/HoneyFiberIncident_Eating_Loop_01}}
 {id=WavePlayer_1, properties {bLooping=true, soundWaveAssetPtr=/Game/Sounds/Fort_Default_Sounds/Player_Healing_Loop}}
 {id=WavePlayer_3, properties {soundWaveAssetPtr=/HoneyFiber_Incident/Assets/Audio/waves/HoneyFiber_Player_Heal_Loop_Start_01}}
]
properties {
 attenuationSettings=&Fort_Consumable_Use_Athena_Att
 maxDistance=2250
 soundClassObject=&Fort_Consumables
 volumeMultiplier=1.5
}
