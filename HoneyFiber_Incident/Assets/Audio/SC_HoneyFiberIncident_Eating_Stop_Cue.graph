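# Sound cue for the stop sound of eating the HoneyFiber Incident consumable.
# A Random node picks one of three eating-stop waves with equal weight and feeds it through a Modulator at the root.
# It plays in the consumables sound class with consumable-use attenuation and a raised volume multiplier.
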SoundCue /HoneyFiber_Incident/Assets/Audio/SC_HoneyFiberIncident_Eating_Stop_Cue
root=Modulator_1
nodes [
 {id=Modulator_1, children [Random_0]}
 {id=Random_0, children [WavePlayer_3,WavePlayer_4,WavePlayer_5], properties {weights [1,1,1]}}
 {id=WavePlayer_3, properties {soundWaveAssetPtr=/HoneyFiber_Incident/Assets/Audio/waves/HoneyFiberIncident_Eating_Stop_01}}
 {id=WavePlayer_4, properties {soundWaveAssetPtr=/HoneyFiber_Incident/Assets/Audio/waves/HoneyFiberIncident_Eating_Stop_02}}
 {id=WavePlayer_5, properties {soundWaveAssetPtr=/HoneyFiber_Incident/Assets/Audio/waves/HoneyFiberIncident_Eating_Stop_03}}
]
properties {
 attenuationSettings=&Fort_Consumable_Use_Athena_Att
 maxDistance=2250
 soundClassObject=&Fort_Consumables
 volumeMultiplier=2
}
