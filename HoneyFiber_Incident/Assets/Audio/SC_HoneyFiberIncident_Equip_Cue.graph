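# Sound cue for the equip sound of the HoneyFiber Incident item, classed under Fort_Consumables.
# A Random node picks one of three equip wave variations with equal weight to vary repeated plays.
# It plays at reduced volume and uses the shared generic player-equip attenuation asset for spatial falloff.
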SoundCue /HoneyFiber_Incident/Assets/Audio/SC_HoneyFiberIncident_Equip_Cue
root=Random_0
nodes [
 {id=Random_0, children [WavePlayer_0,WavePlayer_1,WavePlayer_2], properties {weights [1,1,1]}}
 {id=WavePlayer_0, properties {soundWaveAssetPtr=/HoneyFiber_Incident/Assets/Audio/waves/HoneyFiberIncident_Equip_01}}
 {id=WavePlayer_1, properties {soundWaveAssetPtr=/HoneyFiber_Incident/Assets/Audio/waves/HoneyFiberIncident_Equip_02}}
 {id=WavePlayer_2, properties {soundWaveAssetPtr=/HoneyFiber_Incident/Assets/Audio/waves/HoneyFiberIncident_Equip_03}}
]
properties {
 attenuationSettings=/Game/Sounds/Attenuation/Fort_Player_Equip_Generic_Att
 maxDistance=1800
 soundClassObject=&Fort_Consumables
 volumeMultiplier=0.4
}
