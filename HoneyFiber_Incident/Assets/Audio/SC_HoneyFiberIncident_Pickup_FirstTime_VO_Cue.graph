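# Sound cue for the HoneyFiber Incident pickup, named as the first-time variant, pairing a pickup sound with a voice line.
# A mixer layers one of three random pickup waves with one of three random CheesyPoofs VO lines, all equally weighted.
# A short fixed delay offsets the voice line behind the pickup sound.
# It plays at reduced volume under the Fort_Consumables sound class with consumable-use attenuation.
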SoundCue /HoneyFiber_Incident/Assets/Audio/SC_HoneyFiberIncident_Pickup_FirstTime_VO_Cue
root=Mixer_0
nodes [
 {id=Delay_0, children [Random_0], properties {delayMax=0.3, delayMin=0.3}}
 {id=Mixer_0, children [Random_1,Delay_0], properties {inputVolume [1,1]}}
 {id=Random_0, children [WavePlayer_1,WavePlayer_2,WavePlayer_3], properties {weights [1,1,1]}}
 {id=Random_1, children [WavePlayer_4,WavePlayer_5,WavePlayer_6], properties {weights [1,1,1]}}
 {id=WavePlayer_1, properties {soundWaveAssetPtr=/HoneyFiber_Incident/Assets/Audio/VO/IP_VO_CheesyPoofs_01}}
 {id=WavePlayer_2, properties {soundWaveAssetPtr=/HoneyFiber_Incident/Assets/Audio/VO/IP_VO_CheesyPoofs_02}}
 {id=WavePlayer_3, properties {soundWaveAssetPtr=/HoneyFiber_Incident/Assets/Audio/VO/IP_VO_CheesyPoofs_03}}
 {id=WavePlayer_4, properties {soundWaveAssetPtr=/HoneyFiber_Incident/Assets/Audio/waves/HoneyFiberIncident_PickUP_01}}
 {id=WavePlayer_5, properties {soundWaveAssetPtr=/HoneyFiber_Incident/Assets/Audio/waves/HoneyFiberIncident_PickUP_02}}
 {id=WavePlayer_6, properties {soundWaveAssetPtr=/HoneyFiber_Incident/Assets/Audio/waves/HoneyFiberIncident_PickUP_03}}
]
properties {
 attenuationSettings=&Fort_Consumable_Use_Athena_Att
 maxDistance=2250
 soundClassObject=&Fort_Consumables
 volumeMultiplier=0.5
}
