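# Sound cue for the grab-start sound of the HoneyFiber_Incident throwable, classed under gadget item sounds.
# A random node picks one of three grab waves and feeds a modulator that slightly varies pitch and volume.
# Inline attenuation overrides add a custom falloff curve, low-pass occlusion and a distance-based send to distant reverb.
# A concurrency override caps it at one instance, and a further submix send targets music audio analysis.
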SoundCue /HoneyFiber_Incident/Assets/Audio/SC_HoneyFiber_Incident_Throw_Grab
root=Modulator_1
nodes [
 {id=Modulator_1, children [Random_1], properties {pitchMax=1, pitchMin=0.9, volumeMax=0.9, volumeMin=0.85}}
 {id=Random_1, children [WavePlayer_13, WavePlayer_10, WavePlayer_11], properties {weights [1,1,1]}}
 {id=WavePlayer_10, properties {soundWaveAssetPtr=/HoneyFiber_Incident/Assets/Audio/waves/HoneyFiberIncident_Grab_Start_01}}
 {id=WavePlayer_11, properties {soundWaveAssetPtr=/HoneyFiber_Incident/Assets/Audio/waves/HoneyFiberIncident_Grab_Start_02}}
 {id=WavePlayer_13, properties {soundWaveAssetPtr=/HoneyFiber_Incident/Assets/Audio/waves/HoneyFiberIncident_Grab_Start_04}}
]
properties {
 attenuationSettings=/Game/Sounds/Attenuation/Fort_GadgetThrow_Att
 bOverrideAttenuation=true
 bOverrideConcurrency=true
 maxDistance=2700
 soundClassObject=&Fort_Gadgets_Items
 volumeMultiplier=0.5
 attenuationOverrides {
  bEnableOcclusion=true
  bEnableSubmixSends=true
  distanceAlgorithm=Custom
  falloffDistance=2500
  occlusionLowPassFilterFrequency=500
  occlusionTraceChannel=ECC_GameTraceChannel11
  attenuationShapeExtents {x=200}
  customAttenuationCurve {editorCurveData {keys [
     <at 0 1 RCIM_Cubic>
     {arriveTangent=-1.1458, interpMode=RCIM_Cubic, leaveTangent=-1.1458, time=0.2554, value=0.3005}
     <at 0.8153 0.0659>
     {arriveTangent=-0.7405, interpMode=RCIM_Cubic, leaveTangent=-0.7405, tangentMode=RCTM_User, time=1, value=0}
    ]}}
  submixSendSettings [{
    bEnableHPFCutoff=false
    bEnableLPFCutoff=false
    disableManualSendClamp=false
    hPFCutoff=20
    lPFCutoff=20000
    maxSendDistance=2500
    maxSendLevel=0.1
    minSendDistance=500
    minSendLevel=0
    sendLevel=0.2
    sendLevelControlMethod=Linear
    soundSubmix=&Fortnite_Reverb_Distant
   }]
 }
 concurrencyOverrides {maxCount=1}
 concurrencySet [/Game/Sounds/Concurrency_Rules/Athena_EmoteFoley_Conc]
 soundSubmixSends [{minSendLevel=0, soundSubmix=&Fortnite_AudioAnalysis_Music}]
}
