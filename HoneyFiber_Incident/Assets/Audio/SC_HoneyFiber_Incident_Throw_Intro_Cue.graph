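# Sound cue for the HoneyFiber Incident throw intro, played quietly in the gadget and item sound class.
# An equal-weight random node picks one of three reused "Consumable Finished" waves, and a modulator applies a fixed, slightly lowered pitch.
# Overridden attenuation adds a custom falloff curve, occlusion low-pass and distance-scaled reverb send.
# An emote foley concurrency set governs voice limits, and a submix send feeds music audio analysis.
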SoundCue /HoneyFiber_Incident/Assets/Audio/SC_HoneyFiber_Incident_Throw_Intro_Cue
root=Modulator_1
nodes [
 {id=Modulator_1, children [Random_0], properties {pitchMax=0.9, pitchMin=0.9}}
 {id=Random_0, children [WavePlayer_2,WavePlayer_4,WavePlayer_5], properties {weights [1,1,1]}}
 {id=WavePlayer_2, properties {soundWaveAssetPtr=/HoneyFiber_Incident/Assets/Audio/waves/HoneyFiberIncident_Consumable_Finished_03}}
 {id=WavePlayer_4, properties {soundWaveAssetPtr=/HoneyFiber_Incident/Assets/Audio/waves/HoneyFiberIncident_Consumable_Finished_02}}
 {id=WavePlayer_5, properties {soundWaveAssetPtr=/HoneyFiber_Incident/Assets/Audio/waves/HoneyFiberIncident_Consumable_Finished_01}}
]
properties {
 attenuationSettings=/Game/Sounds/Attenuation/Fort_GadgetThrow_Att
 bOverrideAttenuation=true
 maxDistance=2700
 soundClassObject=&Fort_Gadgets_Items
 volumeMultiplier=0.3
 attenuationOverrides {
  bEnableOcclusion=true
  bEnableSubmixSends=true
  distanceAlgorithm=Custom
  falloffDistance=2500
  occlusionLowPassFilterFrequency=500
  occlusionTraceChannel=ECC_GameTraceChannel11
  reverbDistanceMax=1600
  reverbDistanceMin=200
  reverbWetLevelMax=0.1
  reverbWetLevelMin=0
  attenuationShapeExtents {x=200}
  customAttenuationCurve {editorCurveData {keys [
     <at 0 1 RCIM_Cubic>
     {arriveTangent=-1.1458, interpMode=RCIM_Cubic, leaveTangent=-1.1458, time=0.2554, value=0.3005}
     <at 0.8153 0.0659>
     {arriveTangent=-0.7405, interpMode=RCIM_Cubic, leaveTangent=-0.7405, tangentMode=RCTM_User, time=1, value=0}
    ]}}
  submixSendSettings [{
    bEnableHPFCutoff=false
    bEnableLPFCutoff=false
    disableManualSendClamp=false
    hPFCutoff=20
    lPFCutoff=20000
    maxSendDistance=1600
    maxSendLevel=0.1
    minSendDistance=200
    minSendLevel=0
    sendLevel=0.2
    sendLevelControlMethod=Linear
    soundSubmix=&Fortnite_Reverb_Distant
   }]
 }
 concurrencyOverrides {maxCount=3, resolutionRule=StopOldest}
 concurrencySet [/Game/Sounds/Concurrency_Rules/Athena_EmoteFoley_Conc]
 soundSubmixSends [{minSendLevel=0, soundSubmix=&Fortnite_AudioAnalysis_Music}]
}
