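# Sound cue for the HoneyFiber_Incident throw outro, classed under gadget and item sounds.
# A Random node picks one of three equally weighted waves and passes it through a Modulator.
# Overridden attenuation replaces the assigned gadget-throw preset with a custom falloff curve, occlusion, distance filtering and a distant-reverb send.
# It shares the Athena emote-foley concurrency set and also has a send to the music audio-analysis submix.
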SoundCue /HoneyFiber_Incident/Assets/Audio/SC_HoneyFiber_Incident_Throw_Outro_Cue
root=Modulator_1
nodes [
 {id=Modulator_1, children [Random_1]}
 {id=Random_1, children [WavePlayer_2,WavePlayer_4,WavePlayer_5], properties {weights [1,1,1]}}
 {id=WavePlayer_2, properties {soundWaveAssetPtr=/HoneyFiber_Incident/Assets/Audio/waves/HoneyFiberIncident_Throw_Outro_01}}
 {id=WavePlayer_4, properties {soundWaveAssetPtr=/HoneyFiber_Incident/Assets/Audio/waves/HoneyFiberIncident_Throw_Outro_02}}
 {id=WavePlayer_5, properties {soundWaveAssetPtr=/HoneyFiber_Incident/Assets/Audio/waves/HoneyFiberIncident_Throw_Outro_03}}
]
properties {
 attenuationSettings=/Game/Sounds/Attenuation/Fort_GadgetThrow_Att
 bOverrideAttenuation=true
 maxDistance=2097152
 soundClassObject=&Fort_Gadgets_Items
 attenuationOverrides {
  bEnableOcclusion=true
  bEnableSubmixSends=true
  distanceAlgorithm=Custom
  falloffDistance=3500
  hPFFrequencyAtMax=125
  lPFFrequencyAtMax=1200
  lPFRadiusMax=1500
  lPFRadiusMin=300
  occlusionTraceChannel=ECC_GameTraceChannel11
  attenuationShapeExtents {x=800}
  customAttenuationCurve {editorCurveData {keys [
     <at 0 1 RCIM_Cubic>
     {arriveTangent=-1.4755, interpMode=RCIM_Cubic, leaveTangent=-1.4755, time=0.2663, value=0.8021}
     {arriveTangent=-1.8432, interpMode=RCIM_Cubic, leaveTangent=-1.8432, tangentMode=RCTM_User, time=0.379, value=0.4408}
     {arriveTangent=-0.4002, interpMode=RCIM_Cubic, leaveTangent=-0.4002, tangentMode=RCTM_User, time=0.928, value=0.0939}
     {arriveTangent=-0.7405, interpMode=RCIM_Cubic, leaveTangent=-0.7405, tangentMode=RCTM_User, time=1, value=0.0145}
    ]}}
  submixSendSettings [{
    bEnableHPFCutoff=false
    bEnableLPFCutoff=false
    disableManualSendClamp=false
    hPFCutoff=20
    lPFCutoff=20000
    maxSendDistance=2500
    maxSendLevel=0.1
    minSendDistance=500
    minSendLevel=0
    sendLevel=0.2
    sendLevelControlMethod=Linear
    soundSubmix=&Fortnite_Reverb_Distant
   }]
 }
 concurrencyOverrides {maxCount=3, resolutionRule=StopOldest}
 concurrencySet [/Game/Sounds/Concurrency_Rules/Athena_EmoteFoley_Conc]
 soundSubmixSends [{minSendLevel=0, soundSubmix=&Fortnite_AudioAnalysis_Music}]
}
