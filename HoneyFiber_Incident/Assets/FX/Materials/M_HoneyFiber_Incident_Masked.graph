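# Unlit masked material for Niagara sprites that draws a procedural crescent-like shape, without textures.
# A centered sphere mask minus two offset masks is stepped to a hard shape, then an offset highlight mask is added.
# That result is tinted by particle color and a morning/day/evening/night color from the Color_Time_of_Day function.
# A second copy, widened by OutlineWidth, is dimmed and maxed with the shape to form an outlined mask.
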Material /HoneyFiber_Incident/Assets/FX/Materials/M_HoneyFiber_Incident_Masked
graph {
 properties {
  bAutomaticallySetUsageInEditor=false
  bCastRayTracedShadows=false
  bUsedWithNiagaraSprites=true
  blendMode=BLEND_Masked
  shadingModel=MSM_Unlit
  usedShadingModels=Unlit
 }
 nodes [
  Constant4Vector() {constant=<rgba 1 0.25 0 0>}
  n1 = VectorParameter() {group=Time Of Day, parameterName=Color Morning Multiplier, defaultValue=<rgba 0.9896 0.6927 0.7852 0>}
  n2 = VectorParameter() {group=Time Of Day, parameterName=Color Day Multiplier, defaultValue=<rgba 1 1 1 0>}
  n3 = VectorParameter() {group=Time Of Day, parameterName=Color Evening Multiplier, defaultValue=<rgba 0.7 0.49 0.5211 0>}
  n4 = VectorParameter() {group=Time Of Day, parameterName=Color Night Multiplier, defaultValue=<rgba 0.3498 0.4864 0.6458 0>}
  n5 = SphereMask(
   A=Add(A=TextureCoordinate(), B=Constant2Vector() {r=0.3})
   B=Constant() {r=0.5}
   Radius=ScalarParameter() {defaultValue=0.5, parameterName=Radius}
  )
  n6 = ScalarParameter() {defaultValue=0.6, parameterName=HighlightIntensity}
  n7 = TextureCoordinate()
  n8 = Constant2Vector() {r=0.45}
  n9 = Subtract(
   A=SphereMask(
    A=TextureCoordinate()
    B=Constant() {r=0.5}
    Radius=ScalarParameter() {defaultValue=0.5, parameterName=Radius}
    Hardness=ScalarParameter() {defaultValue=0.6, parameterName=Hardness}
   )
   B=SphereMask(A=Add(A=@n7, B=@n8), B=Constant() {r=0.5}) {attenuationRadius=0.5, hardnessPercent=0.5}
  )
  n10 = SphereMask(A=Add(A=TextureCoordinate(), B=Constant2Vector() {r=0.25}), B=Constant() {r=0.5}) {attenuationRadius=0.5, hardnessPercent=0.5}
  n11 = Step(X=Saturate(None=Subtract(A=@n9, B=@n10))) {constX=0.2, constY=0.5}
  Multiply(
   A=Multiply(
    A=MaterialFunctionCall(Color_Morning=@n1.RGB, Color_Day=@n2.RGB, Color_Evening=@n3.RGB, Color_Night=@n4.RGB) {materialFunction=&Color_Time_of_Day}.Result
    B=ParticleColor().RGB
   )
   B=Add(A=Multiply(A=@n5, B=@n6), B=@n11)
  )
  n12 = ScalarParameter() {defaultValue=0.5, parameterName=Radius}
  n13 = ScalarParameter() {defaultValue=0.04, parameterName=OutlineWidth}
  n14 = TextureCoordinate()
  n15 = Constant2Vector() {r=0.45}
  n16 = Subtract(A=@n12, B=@n13)
  n17 = Subtract(
   A=SphereMask(
    A=TextureCoordinate()
    B=Constant() {r=0.5}
    Radius=Add(A=@n12, B=@n13)
    Hardness=ScalarParameter() {defaultValue=0.6, parameterName=Hardness}
   )
   B=SphereMask(A=Add(A=@n14, B=@n15), B=Constant() {r=0.5}, Radius=@n16) {attenuationRadius=0.49, hardnessPercent=0.5}
  )
  n18 = SphereMask(A=Add(A=TextureCoordinate(), B=Constant2Vector() {r=0.25}), B=Constant() {r=0.5}, Radius=@n16) {attenuationRadius=0.49, hardnessPercent=0.5}
  n19 = Step(X=Saturate(None=Subtract(A=@n17, B=@n18))) {constX=0.2, constY=0.5}
  Multiply(A=Max(A=@n11, B=Multiply(A=@n19) {constB=0.6})) {constB=0.9}
 ]
}
