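# Niagara trail effect for the HoneyFiber Incident handful, apparently a brief scatter of small orange-tinted sprite poofs.
# The Poofs emitter uses a probability-gated spawn rate over a short loop, giving each sprite varied color and an initial velocity.
# Gravity pulls the sprites down while their curve-damped random spin slows and a size curve holds, then shrinks them to zero.
# Two further emitters, LWEBall and LWEBall001, list no renderers or modules.
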NiagaraSystem /HoneyFiber_Incident/Assets/FX/NS_HoneyFiber_Incident_HandfulTrail
dependencies {
 usedDataInterfaces [&NiagaraDataInterfaceCurve, &NiagaraDataInterfaceColorCurve]
 usedDynamicInputs [&ColorFromCurve]
 usedModules [
  &SystemState
  &self:SystemSpawnScript.SystemScriptSource.SystemScriptGraph.NiagaraNodeAssignment_0.SetVariables_#1
 ]
}
emitters {
 LWEBall {rendererClasses [], renderers []}
 LWEBall001 {rendererClasses [], renderers []}
 Poofs {
  emitterUpdateScript [
   &EmitterState(Life Cycle Mode=::NewEnumerator1, Loop Behavior=::NewEnumerator1, Loop Duration=0.6, UseLoopDelay=false)
   &SpawnRate(Spawn Probability=0.8, SpawnRate=12, Use Spawn Probability=true)
  ]
  particleSpawnScript [
   &InitializeParticle(
    Color Maximum=<rgba 0.7011 0.2874 0.0844 1>
    Color Minimum=<rgba 0.7529 0.2747 0.0369 1>
    Color Mode=::NewEnumerator2
    Sprite Rotation Mode=::NewEnumerator2
    Sprite Size=<7.7 11>
    Sprite Size Mode=::NewEnumerator4
    UsePositionOffset=false
    Write Material Random=true
   )
   &AddVelocity(Module.Use Falloff Distance=false, Override Seed=false, Velocity Mode=::NewEnumerator1, Velocity Speed=125)
   &ApplyInitialForces()
  ]
  particleUpdateScript [
   &ParticleState()
   &ScaleSpriteSize(
    Uniform Curve Scale=0.8
    Uniform Curve Sprite Scale=DataInterface() {propertyValues {Curve {
       defaultValue=0
       keys [
        {tangentMode=RCTM_User, time=0, value=1}
        {arriveTangent=-0.0405, interpMode=RCIM_Cubic, leaveTangent=-0.0405, tangentMode=RCTM_User, time=0.7009, value=1}
        {arriveTangent=-8.0089, interpMode=RCIM_Cubic, leaveTangent=-8.0089, tangentMode=RCTM_User, time=1, value=0}
       ]
      }}}
   )
   &SpriteRotationRate(Rotation Rate=&Multiply_Float(
     A=&RandomRangeFloat(Minimum=2, Maximum=4)
     B=&FloatFromCurve(FloatCurve=DataInterface() {propertyValues {Curve {keys [<at 0 1>,<at 1 0.5>]}}})
    ))
   &GravityForce(Gravity=<0 0 -700>)
   &SolveForcesAndVelocity(Clamp Velocity=false, Limit Acceleration=false)
  ]
  rendererClasses [&NiagaraSpriteRendererProperties]
  renderers [{rendererClass=&NiagaraSpriteRendererProperties}]
 }
}
summary {
 systemName=NS_HoneyFiber_Incident_HandfulTrail
 emitters [
  {emitterName=Poofs, rendererClasses [&NiagaraSpriteRendererProperties]}
  {emitterName=LWEBall, rendererClasses []}
  {emitterName=LWEBall001, rendererClasses []}
 ]
 userVariables []
}
