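# Niagara effect for an orange, honey-colored projectile in flight, with its body, glow and trail.
# A mesh and burst sprites form the body, a flare emitter adds sprites, a light and a mesh, and smoke, distance-spawned falling poofs and a three-renderer ribbon trail follow.
# User.Hit kills the body and flare particles and stops ribbon spawning.
# Time-of-day values tint most layers, and camera distance scales parts of the flare.
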NiagaraSystem /HoneyFiber_Incident/Assets/FX/NS_HoneyFiber_Incident_ProjectileTrail
dependencies {
 usedDataInterfaces [&NiagaraDataInterfaceCurve, &NiagaraDataInterfaceColorCurve, &NiagaraDataInterfaceCamera]
 usedDynamicInputs [/Niagara/DynamicInputs/Helpers/SimulationPosition, &FNTimeOfDayScalar]
 usedModules [
  &SystemState
  &self:SystemSpawnScript.SystemScriptSource.SystemScriptGraph.NiagaraNodeAssignment_1.SetVariables_#1
  &self:SystemSpawnScript.SystemScriptSource.SystemScriptGraph.NiagaraNodeAssignment_0.SetVariables_#2
 ]
}
emitters {
 Airglow_Light_Flare {
  emitterUpdateScript [
   /SetVariables(Emitter.TODColorMult:LinearColor=&FNTimeOfDayColor(Night Color=<rgba 1 1 1 1>, Morning Color=<rgba 1.25 1.25 1.25 1>, Day Color=<rgba 1.5 1.5 1.5 1>, Evening Color=<rgba 1.25 1.25 1.25 1>)) {isSetParameters=true}
   &EmitterState(
    Inactive Response=::NewEnumerator1
    Life Cycle Mode=::NewEnumerator1
    Loop Behavior=::NewEnumerator1
    Loop Duration Mode=::NewEnumerator1
    UseLoopDelay=false
   )
   &SpawnBurst_Instantaneous(Use Spawn Probability=false, UseLoopCountLimit=false)
   /SetVariables(
    Emitter.LightColor:LinearColor=&ColorFromCurve(
     ColorCurve=DataInterface() {propertyValues {
       AlphaCurve {keys [<at 0.0513 1 RCIM_Cubic>]}
       BlueCurve {keys [<at 0.084 0>,<at 0.2843 0 RCIM_Cubic>]}
       GreenCurve {keys [<at 0.084 0>,<at 0.2843 3.7041 RCIM_Cubic>]}
       RedCurve {keys [<at 0.084 0>,<at 0.2843 25 RCIM_Cubic>]}
      }}
     CurveIndex=$Emitter.Age
    )
    Emitter.LightExponent=1.5
    Emitter.LightRadius:NiagaraFloat=&Lerp_Float(
     A=0
     B=180
     Alpha=&FloatFromCurve(
      FloatCurve=DataInterface() {propertyValues {Curve {keys [{interpMode=RCIM_Cubic, leaveTangent=1.7761, tangentMode=RCTM_Break, time=0.1, value=0}, <at 1 1>]}}}
      CurveIndex=$Emitter.Age
     )
    )
   ) {isSetParameters=true}
  ]
  particleSpawnScript [
   &PartitionParticles()
   &InitializeParticle(
    Color=<rgba 1 0.1667 0 0.5>
    Lifetime=0.17
    Position Offset=<0 1 0>
    Sprite Size Mode=::NewEnumerator3
    Uniform Sprite Size=150
    UsePositionOffset=true
    Write Material Random=true
   )
   &CameraOffset()
  ]
  particleUpdateScript [
   &ParticleState(Kill Particles When Lifetime Has Elapsed=false, Let Infinitely Lived Particles Die When Emitter Deactivates=true)
   &ScaleSpriteSize(Scale Sprite Size Mode=::NewEnumerator0)
   &FN_Sprite_Attributes(
    Color=&MultiplyLinearColor(
     Linear Color=&MultiplyLinearColor(
      Linear Color=<rgba 1 0.25 0 2>
      Scale Factor=&MakeVector4FromVectorAndFloat("Vector (RGB)"=&VectorFromFloat(Value=0.3), "Float (Alpha)"=2)
     )
     Scale Factor=&MakeVector4_FromLinearColor(LinearColor=$Emitter.TODColorMult)
    )
    Create Camera Offset Attribute=true
    Create Color Attribute=true
    "Create Normalized Age "=true
    Create SpriteSize Attribute=true
    SpriteSize=&Multiply_Vector2D_ByFloat(
     Float=&Multiply_Float(
      A=&FloatFromCurve(FloatCurve=DataInterface() {propertyValues {Curve {keys [
           {interpMode=RCIM_Cubic, leaveTangent=4.0017, tangentMode=RCTM_Break, time=0, value=0}
           {arriveTangent=0.2116, interpMode=RCIM_Cubic, tangentMode=RCTM_Break, time=0.5, value=1.55}
          ]}}})
      B=/Niagara/DynamicInputs/Camera/CameraFade(
       Camera Query:NiagaraDataInterfaceCamera=DataInterface()
       Fade Offset:NiagaraFloat=0
       Fade Distance:NiagaraFloat=1000
       Use Owner Position:NiagaraBool=true
      )
     )
     Vector2D=<150 150>
    )
    Write to Module Namespace=true
   )
   &FN_Light_Attributes(
    Color=&MultiplyLinearColor(
     Linear Color=<rgba 1 0.3872 0 1>
     Scale Factor=&MakeVector4_FromLinearColor(LinearColor=$Emitter.TODColorMult)
    )
    Create Color Attribute=true
    Create Exponent Attribute=true
    Create Radius Attribute=true
    Enable Particle Light Enable=true
    Exponent=5
    Light Enabled=true
    Radius=500
    Scale Color=5
   )
   &FN_Sprite_Attributes(
    Camera Offset:NiagaraFloat=&Lerp_Float(
     A=0
     B=10
     Alpha=&FN_LOD_Distance_Remap(Start Distance=0, End Distance=10000, Minimum Output=0, Maximum Output=1, Clamp Output=true)
    )
    Color=&MultiplyLinearColor(
     Linear Color=&MultiplyLinearColor(
      Linear Color=&ColorFromCurve(ColorCurve=DataInterface() {propertyValues {
         AlphaCurve {keys [<at 0.0006 0>,<at 1 1 RCIM_Cubic>]}
         BlueCurve {keys [<at 0 0 RCIM_Cubic>]}
         GreenCurve {keys [<at 0 0.25 RCIM_Cubic>]}
         RedCurve {keys [<at 0 1 RCIM_Cubic>]}
        }})
      Scale Factor=&MakeVector4FromVectorAndFloat("Vector (RGB)"=&VectorFromFloat(Value=1), "Float (Alpha)"=0.3)
     )
     Scale Factor=&MakeVector4_FromLinearColor(LinearColor=$Emitter.TODColorMult)
    )
    Create Camera Offset Attribute=true
    Create Color Attribute=true
    Create SpriteSize Attribute=true
    SpriteSize=&Multiply_Vector2D_ByFloat(
     Float=&Lerp_Float(
      A=0
      B=1
      Alpha=&FN_LOD_Distance_Remap(Start Distance=0, End Distance=5000, Minimum Output=0.01, Maximum Output=1, Clamp Output=true)
     )
     Vector2D=&Multiply_Vector2D_ByFloat(
      Float=&RandomRangeFloat(Minimum=1, Maximum=1.5, Evaluation Type=::NewEnumerator1)
      Vector2D=&Multiply_Vector2D_ByFloat(Float=1.3, Vector2D=<300 5>)
     )
    )
    Write to Module Namespace=true
   ) {module=FN_Sprite_Attributes001}
   &FN_Mesh_Attributes(
    Color=&MultiplyLinearColor(Linear Color=<rgba 0.8 0.2 0 2>, Scale Factor=&MakeVector4_FromLinearColor(LinearColor=$System.ToDColor))
    Create Color Attribute=true
    Create Orientation Attribute=true
    Create Scale Attribute=true
    Orientation:Quat4f=&MultiplyQuaternion(
     Quaternion A=<0 0 0 1>
     Quaternion B=&MakeQuaternion(Quaternion From=0, Axis=<0 0 1>, Angle=0.5, From=?, To=?, Angle Type=::NewEnumerator1, Normalize Vectors=false)
    )
    Scale=&Multiply_VectorByFloat(Vector=<0.2 0.25 0.25>, Float=9)
    Write to Module Namespace=true
   )
   /SetVariables(Particles.JetPosition:NiagaraPosition=&AddVectorToPosition(Position=$Engine.Emitter.SimulationPosition, Vector=<0 5 0>)) {enabled=false, isSetParameters=true}
   &KillParticles(Kill Particles=$User.Hit)
  ]
  rendererClasses [
   &NiagaraSpriteRendererProperties
   &NiagaraLightRendererProperties
   &NiagaraMeshRendererProperties
  ]
  renderers [
   {rendererClass=&NiagaraSpriteRendererProperties}
   {rendererClass=&NiagaraLightRendererProperties, rendererIndex=1}
   {rendererClass=&NiagaraSpriteRendererProperties, rendererIndex=2}
   {rendererClass=&NiagaraMeshRendererProperties, rendererIndex=3}
  ]
 }
 Poofs {
  emitterUpdateScript [
   &EmitterState(Life Cycle Mode=::NewEnumerator1, Loop Duration=1, UseLoopDelay=false)
   &SpawnPerUnit(
    Max Movement Threshold=50000
    Module.Use Max Movement Threshold=true
    Module.Use Movement Tolerance=true
    Movement Tolerance=100
    Spawn Spacing=100
    Use Spawn Probability=false
   )
  ]
  particleSpawnScript [
   &InitializeParticle(
    Color=&RandomRangeLinearColor(Minimum=<rgba 0.7529 0.2747 0.0369 1>, Maximum=<rgba 0.7011 0.2874 0.0844 1>)
    Lifetime=&RandomRangeFloat(Minimum=0.6, Maximum=1.2)
    Sprite Rotation Mode=::NewEnumerator2
    Sprite Size=&Multiply_Vector2D_ByFloat(Float=&RandomRangeFloat(Minimum=1.2, Maximum=1.4), Vector2D=<7.5 10>)
    Sprite Size Mode=::NewEnumerator4
    UsePositionOffset=false
    Write Material Random=true
   )
   &AddVelocity(Module.Use Falloff Distance=false, Override Seed=false, Velocity Mode=::NewEnumerator1, Velocity Speed=125)
   &ShapeLocation(MaskSpawnGroup=false, Override Seed=false, Sphere Radius=5, UseOwnerScale=true)
   &ApplyInitialForces()
  ]
  particleUpdateScript [
   &ParticleState()
   &ScaleSpriteSize(Uniform Curve Sprite Scale=DataInterface() {propertyValues {Curve {
       defaultValue=0
       keys [
        {tangentMode=RCTM_User, time=0, value=1}
        {arriveTangent=-0.0405, interpMode=RCIM_Cubic, leaveTangent=-0.0405, tangentMode=RCTM_User, time=0.7, value=1}
        {arriveTangent=-8.1175, interpMode=RCIM_Cubic, leaveTangent=-8.1175, tangentMode=RCTM_User, time=1, value=0}
       ]
      }}})
   &SpriteRotationRate(Rotation Rate=&Multiply_Float(
     A=&RandomRangeFloat(Minimum=200, Maximum=600)
     B=&FloatFromCurve(FloatCurve=DataInterface() {propertyValues {Curve {keys [<at 0 1>,<at 1 0.5>]}}})
    ))
   &GravityForce(Gravity=&Multiply_VectorByFloat(Vector=<1 1 -600>, Float=&RandomRangeFloat(Minimum=0.8, Maximum=1.2)))
   &Collision(
    Age Colliding Particles=false
    Control Roll On Collision=true
    Friction=0.3
    Friction During a Bounce=0.2
    KillOnCollision=false
    Module.Control Rest State Via Surface Normal=false
    Module.EnableMaxCollisionCount=false
    Module.Randomize Collision Normal Vector=true
    Restitution=0.3
    ScaleRateOfRotation=true
   ) {enabled=false}
   &SolveForcesAndVelocity(Clamp Velocity=false, Limit Acceleration=false)
  ]
  rendererClasses [&NiagaraSpriteRendererProperties]
  renderers [{rendererClass=&NiagaraSpriteRendererProperties}]
 }
 Projectile {
  emitterUpdateScript [
   &EmitterState(Life Cycle Mode=::NewEnumerator1, Loop Duration=1, UseLoopDelay=false)
   &SpawnBurst_Instantaneous(Spawn Count=8, Use Spawn Probability=false, UseLoopCountLimit=false)
  ]
  particleSpawnScript [
   &InitializeParticle(
    Color=&RandomRangeLinearColor(Minimum=<rgba 0.7529 0.2747 0.0369 1>, Maximum=<rgba 0.7011 0.2685 0.0561 1>)
    Sprite Rotation Mode=::NewEnumerator2
    Sprite Size=<12 16>
    Sprite Size Mode=::NewEnumerator4
    UsePositionOffset=false
    Write Material Random=true
   )
   &ShapeLocation(MaskSpawnGroup=false, Override Seed=false, Sphere Radius=8, Sphere Surface Distribution=0.8214, UseOwnerScale=true)
   &CameraOffset(Camera Offset Amount=6)
  ]
  particleUpdateScript [
   &ParticleState()
   &KillParticles(Kill Particles=$User.Hit)
  ]
  rendererClasses [&NiagaraSpriteRendererProperties]
  renderers [{rendererClass=&NiagaraSpriteRendererProperties}]
 }
 Projectile001 {
  emitterUpdateScript [
   &EmitterState(Life Cycle Mode=::NewEnumerator1, Loop Duration=1, UseLoopDelay=false)
   &SpawnBurst_Instantaneous(Use Spawn Probability=false, UseLoopCountLimit=false)
  ]
  particleSpawnScript [&InitializeParticle(
    Color=<rgba 0.4 0.1458 0.028 1>
    Mesh Scale Mode=::NewEnumerator3
    Mesh Uniform Scale=1.3
    Sprite Rotation Mode=::NewEnumerator2
    Sprite Size Mode=::NewEnumerator3
    Uniform Sprite Size=75
    UsePositionOffset=false
    Write Material Random=true
   )]
  particleUpdateScript [
   &ParticleState()
   &ScaleMeshSize(Scale Factor=&VectorFromFloat(Value=&FloatFromCurve(
      FloatCurve=DataInterface() {propertyValues {Curve {defaultValue=0, keys [{tangentMode=RCTM_User, time=0, value=0}, {tangentMode=RCTM_User, time=0.25, value=1}]}}}
      CurveIndex=$System.Age
     ))) {enabled=false}
   &ScaleColor(Scale RGB=&MakeVectorFromLinearColorRGB(Color=$System.ToDColor), ScaleA=true, ScaleRGB=true)
   /SetVariables(Particles.Shadow Color:LinearColor=&MultiplyLinearColor(
     Linear Color=&FNTimeOfDayColor(
      Night Color=<rgba 0.3 0.129 0.03 1>
      Morning Color=<rgba 0.3 0.129 0.03 1>
      Day Color=<rgba 0.3 0.129 0.03 1>
      Evening Color=<rgba 0.3 0.129 0.03 1>
     )
     Scale Factor=&MakeVector4FromVectorAndFloat("Vector (RGB)"=&VectorFromFloat(Value=1), "Float (Alpha)"=0.75)
    )) {isSetParameters=true}
   &DynamicMaterialParameters(
    Index 0 Param 4=&Add_Float(
     A=&Multiply_Float(A=$Engine.Time, B=0.05)
     B=&RandomRangeFloat(Minimum=0, Maximum=1000)
    )
    Index 0 Vector 1:Vector3f=&MakeVectorFromLinearColorRGB(Color=$Particles.Shadow Color)
    Index 1 Param 1=&MakeFloatFromInt(Int=&RandomRangeInt(Minimum=7, Maximum=5))
    Index 1 Param 2=&FloatFromCurve(FloatCurve=DataInterface() {propertyValues {Curve {keys [<at 0.5 1 RCIM_Cubic>,<at 1 0>]}}}, Scale Curve=2)
    Module.Index0Float3Write=true
    Module.Index0VectorWrite=true
    Module.Index1Float0Write=true
    Module.Index1Float1Write=true
    Module.Index1Float2Write=true
    Module.Index1Float3Write=true
    Write Parameter Index 0=::NewEnumerator3
    Write Parameter Index 1=::NewEnumerator1
   )
   &KillParticles(Kill Particles=$User.Hit)
   /SetVariables(Particles.Dyn1:Vector4f=<3 1 1 1>, Particles.GlowColor=<rgba 0.5 0.15 0 1>) {enabled=false, isSetParameters=true}
  ]
  rendererClasses [&NiagaraMeshRendererProperties]
  renderers [{rendererClass=&NiagaraMeshRendererProperties, rendererIndex=0}]
 }
 Smoke {
  emitterUpdateScript [
   &EmitterState(Life Cycle Mode=::NewEnumerator1, UseLoopDelay=false)
   &SpawnRate(SpawnRate=128, Use Spawn Probability=false)
  ]
  particleSpawnScript [
   &InitializeParticle(
    Color Maximum=<rgba 0.4288 0.2004 0.0214 0.25>
    Color Minimum=<rgba 0.3003 0.085 0.015 0.125>
    Color Mode=::NewEnumerator2
    Lifetime Max=0.6
    Lifetime Min=0.4
    Lifetime Mode=::NewEnumerator1
    Mesh Scale Mode=::NewEnumerator3
    Mesh Uniform Scale=0.2
    Sprite Rotation Mode=::NewEnumerator2
    Sprite Size Mode=::NewEnumerator3
    Uniform Sprite Size=75
    UsePositionOffset=false
    Write Material Random=true
   )
   &ShapeLocation(MaskSpawnGroup=false, Override Seed=false, Sphere Radius=8, UseOwnerScale=true)
   &AddVelocity(Constrain To Radius=150, Module.Use Falloff Distance=false, Override Seed=false, Velocity Mode=::NewEnumerator1)
  ]
  particleUpdateScript [
   &ParticleState()
   &ScaleColor(
    Scale Alpha=&FloatFromCurve(
     FloatCurve=DataInterface() {propertyValues {Curve {keys [{interpMode=RCIM_Cubic, tangentMode=RCTM_User, time=0, value=1}, <at 0.8 0>]}}}
     Scale Curve=&Multiply_Float(A=3, B=$System.ToDScalar)
    )
    Scale RGB=&VectorFromFloat(Value=&FloatFromCurve(
      FloatCurve=DataInterface() {propertyValues {Curve {keys [{arriveTangent=1.8145, interpMode=RCIM_Cubic, leaveTangent=4.5795, tangentMode=RCTM_Break, time=0, value=0.4}, <at 0.3387 0.9565>, <at 1 1>]}}}
      Scale Curve=&Multiply_Float(A=3, B=$System.ToDScalar)
     ))
    ScaleA=true
    ScaleRGB=true
   )
   &ScaleSpriteSize(Scale Factor=&Vector2DFromFloat(Value=&FloatFromCurve(FloatCurve=DataInterface() {propertyValues {Curve {keys [
          {arriveTangent=8.1876, interpMode=RCIM_Cubic, leaveTangent=5.3162, tangentMode=RCTM_User, time=0, value=0.15}
          {arriveTangent=1.0025, interpMode=RCIM_Cubic, leaveTangent=1.0025, tangentMode=RCTM_User, time=0.2, value=0.65}
          {arriveTangent=0.1638, interpMode=RCIM_Cubic, leaveTangent=0.1638, tangentMode=RCTM_User, time=1.0463, value=1}
         ]}}})))
   &SolveForcesAndVelocity(Clamp Velocity=false, Limit Acceleration=false)
  ]
  rendererClasses [&NiagaraSpriteRendererProperties]
  renderers [{rendererClass=&NiagaraSpriteRendererProperties}]
 }
 Trails {
  emitterUpdateScript [
   &EmitterState()
   &SpawnRate(
    Spawn Probability=&MakeCustomFloatFromBool(Bool=$User.Hit, True Float=0, False Float=1)
    SpawnRate=64
    Use Spawn Probability=true
   )
  ]
  particleSpawnScript [&InitializeParticle(
    Color=<rgba 0.5 0.1892 0.025 1>
    Lifetime=0.5
    Ribbon Width=20
    Ribbon Width Mode=::NewEnumerator1
    UsePositionOffset=false
    Write Material Random=true
   )]
  particleUpdateScript [
   &ParticleState()
   &ScaleColor(
    Scale Alpha=&FloatFromCurve(
     FloatCurve=DataInterface() {propertyValues {Curve {keys [{interpMode=RCIM_Cubic, tangentMode=RCTM_User, time=0, value=1}, <at 1 0>]}}}
     Scale Curve=&Multiply_Float(A=30, B=$System.ToDScalar)
    )
    Scale RGB=&MakeVectorFromLinearColorRGB(Color=$System.ToDColor)
    ScaleA=true
    ScaleRGB=true
   )
   &ScaleRibbonWidth(Ribbon Width Scale=&FloatFromCurve(FloatCurve=DataInterface() {propertyValues {Curve {defaultValue=1, keys [<at 0 1>, {arriveTangent=-1, leaveTangent=-1, time=0.2, value=1}, {leaveTangent=-2.5176, time=1, value=0}]}}}))
   /SetVariables(
    Particles.ToonLinesColor:LinearColor=&MultiplyLinearColor(Linear Color=$Particles.Color, Scale Factor=<0.6 0.6 0.6 0.1>)
    Particles.ToonTrailColor:LinearColor=&MakeLinearColorFromVectorAndFloat(
     "Vector (RGB)"=&MakeVectorFromLinearColorRGB(Color=<rgba 1 0.4112 0.1 1>)
     "Float (Alpha)"=&FloatFromCurve(
      FloatCurve=DataInterface() {propertyValues {Curve {keys [{interpMode=RCIM_Cubic, tangentMode=RCTM_User, time=0, value=1}, <at 0.4 0>]}}}
      Scale Curve=&Multiply_Float(A=20, B=$System.ToDScalar)
     )
    )
    Particles.ToonTrailWidth:NiagaraFloat=&Multiply_Float(A=$Particles.RibbonWidth, B=0.4)
   ) {isSetParameters=true}
   /SetVariables(Particles.Position=&Lerp_Vector(
     A=Linked() {linkedVariable {name=Engine.Emitter.SimulationPosition, type=NiagaraPosition}}
     B=Linked() {linkedVariable {name=Particles.Position, type=NiagaraPosition}}
     Alpha=&FloatFromCurve(FloatCurve=DataInterface() {propertyValues {Curve {keys [<at 0 0>,<at 0.1 1>]}}})
    )) {isSetParameters=true}
  ]
  rendererClasses [&NiagaraRibbonRendererProperties]
  renderers [
   {rendererClass=&NiagaraRibbonRendererProperties}
   {rendererClass=&NiagaraRibbonRendererProperties, rendererIndex=1}
   {rendererClass=&NiagaraRibbonRendererProperties, rendererIndex=2}
  ]
 }
}
summary {
 systemName=NS_HoneyFiber_Incident_ProjectileTrail
 emitters [
  {emitterName=Smoke, rendererClasses [&NiagaraSpriteRendererProperties]}
  {emitterName=Trails, rendererClasses [&NiagaraRibbonRendererProperties]}
  {emitterName=Poofs, rendererClasses [&NiagaraSpriteRendererProperties]}
  {emitterName=Projectile, rendererClasses [&NiagaraSpriteRendererProperties]}
  {emitterName=Projectile001, rendererClasses [&NiagaraMeshRendererProperties]}
  {
   emitterName=Airglow_Light_Flare
   rendererClasses [
    &NiagaraSpriteRendererProperties
    &NiagaraLightRendererProperties
    &NiagaraMeshRendererProperties
   ]
  }
 ]
 userVariables [
  User.Hit {type=NiagaraBool}
  User.RibbonIDOffset {type=NiagaraInt32}
 ]
}
